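# Two-sided, masked, dithered Default Lit material for mesh particles, named for the feathers and eggshell of the TenderDefender glider.
# A powered vertical UV gradient blends two color parameters, then tiled sparkle noise, broken up by blurred noise, layers two more tints as speckles.
# TwoSidedSign darkens this color on back faces.
# The graph also holds a noise-driven scalar blend, a constant and a particle color node.
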Material /BRCosmetics/Athena/Effects/Meshes/Gliders/TenderDefender/M_FeathersEggShell
graph {
 properties {
  bUsedWithMeshParticles=true
  blendMode=BLEND_Masked
  ditherOpacityMask=true
  pixelDepthOffsetMode=PDOM_Legacy
  twoSided=true
  usedShadingModels=Default Lit
 }
 nodes [
  Constant() {r=0.25}
  n1 = TextureSample(UVs=TextureCoordinate() {uTiling=3, vTiling=3}) {samplerType=SAMPLERTYPE_Grayscale, texture=&LowResBlurredNoise}
  LinearInterpolate(Alpha=@n1.R) {constA=0.3, constB=0.6}
  n2 = VectorParameter() {parameterName=Color_1, defaultValue=<rgba 1 0.9833 0.75 0>}
  n3 = VectorParameter() {parameterName=Color, defaultValue=<rgba 0.6654 0.6654 0.4658 1>}
  n4 = Power(Base=ComponentMask(None=TextureCoordinate()) {g=true})
  n5 = TextureSample(UVs=TextureCoordinate() {uTiling=5, vTiling=5}) {samplerType=SAMPLERTYPE_LinearColor, texture=/Game/Gadgets/Assets/Slow_Field/Textures/T_Slowfield_SparkleNoise_01}
  n6 = LinearInterpolate(Alpha=@n1.R)
  n7 = TextureCoordinate() {uTiling=15, vTiling=15}
  n8 = TextureCoordinate()
  n9 = LinearInterpolate(
   A=LinearInterpolate(
    A=LinearInterpolate(A=@n2.RGB, B=@n3.RGB, Alpha=@n4)
    B=VectorParameter() {parameterName=Color_2, defaultValue=<rgba 0.2086 0.1651 0.0723 1>}.RGB
    Alpha=Multiply(A=@n5.G, B=@n6)
   )
   B=VectorParameter() {parameterName=Color_3, defaultValue=<rgba 0.3005 0.3515 0.1683 1>}.RGB
   Alpha=Multiply(
    A=TextureSample(UVs=@n7) {samplerType=SAMPLERTYPE_LinearColor, texture=&T_Slowfield_SparkleNoise_01}.R
    B=TextureSample(UVs=@n8) {samplerType=SAMPLERTYPE_Grayscale, texture=&LowResBlurredNoise}.R
   )
  )
  LinearInterpolate(A=Multiply(A=@n9) {constB=0.25}, B=@n9, Alpha=Clamp(None=TwoSidedSign())) {constB=0}
  ParticleColor()
 ]
}
